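# Physics asset for the M_MED_Banner male medium player body, covering only the belt and pelvis area rather than a full ragdoll.
# A box body on the dyn_belt_parent bone and a small capsule on the pelvis bone provide the collision shapes.
# A single constraint joins the two bodies and locks the first swing axis.
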PhysicsAsset /BRCosmetics/Characters/Player/Male/Medium/Bodies/M_MED_Banner/Meshes/M_MED_Banner_Physics
bodies {
 dyn_belt_parent [{shapeType=Box, extentX=12.4464, extentY=21.8894, extentZ=4.3027, center=<-5.9544 0.0003 0>}]
 pelvis [{shapeType=Capsule, length=3.7953, radius=1.6738, center=<1.2771 0.9489 0.1495>, rotation=<rot 3.7255 96.0545 -90>}]
}
constraints [{bone1Name=dyn_belt_parent, bone2Name=pelvis, swing1LimitDegrees=10, swing1Motion=Locked, swing2LimitDegrees=10, twistLimitDegrees=1}]
