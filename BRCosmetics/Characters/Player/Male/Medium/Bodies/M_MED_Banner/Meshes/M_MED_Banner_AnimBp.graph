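# Post-copy AnimBlueprint for the M_MED_Banner character body, animating its dangling belt bones over the parent mesh's pose.
# Thigh Pose Drivers drive belt bones from a leg-movement PoseAsset, and a constraint ties dyn_belt_07 to two neighbors before caching.
# Pelvis-space RigidBody physics fades in while jumping or falling, but not while crouching.
# Crouch, jump, skydive and ADS bone-offset chains are unconnected to output; Belt_ext_force is set but never read.
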AnimBlueprint /BRCosmetics/Characters/Player/Male/Medium/Bodies/M_MED_Banner/Meshes/M_MED_Banner_AnimBp
target_skeleton=&/M_MED_Banner_Skeleton
graphs [
 AnimGraph {nodes [
   n1 = Get Crouching_Alpha() -> (Crouching_Alpha:double)
   n2 = "Transform (Modify) Bone - Bone: dyn_belt_01"(Alpha=@n1, Translation="0,0,3", Rotation="0,0,0", Scale="1,1,1") {
    translationMode=BMM_Additive
    translationSpace=BCS_BoneSpace
    alphaBoolBlend {blendOption=Linear}
    boneToModify {boneName=dyn_belt_01}
    translation {z=3}
   }
   n3 = "Transform (Modify) Bone - Bone: dyn_belt_06"(ComponentPose=@n2, Alpha=@n1, Translation="0,0,4.5", Rotation="0,0,0", Scale="1,1,1") {
    translationMode=BMM_Additive
    translationSpace=BCS_BoneSpace
    alphaBoolBlend {blendOption=Linear}
    boneToModify {boneName=dyn_belt_06}
    translation {z=4.5}
   }
   n4 = "Transform (Modify) Bone - Bone: dyn_belt_05"(ComponentPose=@n3, Alpha=@n1, Translation="0,0,5", Rotation="0,0,0", Scale="1,1,1") {
    translationMode=BMM_Additive
    translationSpace=BCS_BoneSpace
    alphaBoolBlend {blendOption=Linear}
    boneToModify {boneName=dyn_belt_05}
    translation {z=5}
   }
   n5 = "Transform (Modify) Bone - Bone: dyn_belt_02"(ComponentPose=@n4, Alpha=@n1, Translation="0,0,5", Rotation="0,0,0", Scale="1,1,1") {
    translationMode=BMM_Additive
    translationSpace=BCS_BoneSpace
    alphaBoolBlend {blendOption=Linear}
    boneToModify {boneName=dyn_belt_02}
    translation {z=5}
   }
   n6 = "Transform (Modify) Bone - Bone: dyn_belt_03"(ComponentPose=@n5, Alpha=@n1, Translation="0,0,6", Rotation="0,0,0", Scale="1,1,1") {
    translationMode=BMM_Additive
    translationSpace=BCS_BoneSpace
    alphaBoolBlend {blendOption=Linear}
    boneToModify {boneName=dyn_belt_03}
    translation {z=6}
   }
   "Transform (Modify) Bone - Bone: dyn_belt_04"(ComponentPose=@n6, Alpha=@n1, Translation="0,0,6", Rotation="0,0,0", Scale="1,1,1") {
    translationMode=BMM_Additive
    translationSpace=BCS_BoneSpace
    alphaBoolBlend {blendOption=Linear}
    boneToModify {boneName=dyn_belt_04}
    translation {z=6}
   }
   "Transform (Modify) Bone - Bone: dyn_belt_parent"(Alpha=Get Jump_Alpha() -> (Jump_Alpha:double), Translation="2.5,0,3", Rotation="-15,0,-10", Scale="1,1,1") {
    rotationMode=BMM_Additive
    rotationSpace=BCS_BoneSpace
    translationMode=BMM_Additive
    translationSpace=BCS_BoneSpace
    alphaBoolBlend {blendOption=Linear}
    boneToModify {boneName=dyn_belt_parent}
    rotation {pitch=-15, roll=-10}
    translation {x=2.5, z=3}
   }
   n7 = Get bIsSkydiveDiveMode()
   n8 = Get bIsSkydiveDiveMode()
   n9 = Get bIsSkydiveFloating()
   n10 = "Transform (Modify) Bone - Bone: dyn_shoulder_l_scale"(
    Alpha=SelectFloat(A=1, B=0, bPickA=OR Boolean!(A=@n7, B=@n8, C:bool=@n9) -> (ReturnValue))
    Translation="0,0,0"
    Rotation="0,0,0"
    Scale="15,15,15"
   ) {
    scaleMode=BMM_Additive
    scaleSpace=BCS_BoneSpace
    translationMode=BMM_Additive
    translationSpace=BCS_BoneSpace
    alphaBoolBlend {blendOption=Linear}
    boneToModify {boneName=dyn_shoulder_l_scale}
    scale=<15 15 15>
   }
   n11 = Get Ads_Alpha() -> (Ads_Alpha:double)
   n12 = "Transform (Modify) Bone - Bone: dyn_shoulder_l_scale"(ComponentPose=@n10, Alpha=@n11, Translation="0,0,0", Rotation="0,0,0", Scale="3,3,3") {
    scaleMode=BMM_Additive
    scaleSpace=BCS_BoneSpace
    translationMode=BMM_Additive
    translationSpace=BCS_BoneSpace
    alphaBoolBlend {blendOption=Linear}
    boneToModify {boneName=dyn_shoulder_l_scale}
    scale=<3 3 3>
   }
   "Transform (Modify) Bone - Bone: dyn_shoulder_r_scale"(ComponentPose=@n12, Alpha=@n11, Translation="0,0,0", Rotation="0,0,0", Scale="3,3,3") {
    scaleMode=BMM_Additive
    scaleSpace=BCS_BoneSpace
    translationMode=BMM_Additive
    translationSpace=BCS_BoneSpace
    alphaBoolBlend {blendOption=Linear}
    boneToModify {boneName=dyn_shoulder_r_scale}
    scale=<3 3 3>
   }
   n13 = Local To Component(LocalPose=Use cached pose 'BeltCache'())
   n14 = RigidBody(ComponentPose=Local To Component(LocalPose=Component To Local(ComponentPose=@n13)), Alpha=Get Jump_Alpha()) {
    bClampLinearTranslationLimitToRefPose=true
    bOverrideWorldGravity=true
    overridePhysicsAsset=&/M_MED_Banner_Physics
    simulationSpace=BaseBoneSpace
    alphaBoolBlend {blendOption=Linear}
    alphaScaleBiasClamp {bInterpResult=true, interpSpeedIncreasing=7}
    baseBoneRef {boneName=pelvis}
    componentAppliedLinearAccClamp=<4000 4000 4000>
    componentLinearAccScale=<0.4 0.4 0.4>
    componentLinearVelScale=<0.4 0.4 0.4>
    overrideWorldGravity {z=-980}
   }
   n15 = RigidBody(ComponentPose=Local To Component(LocalPose=Use cached pose 'BeltCache'()), Alpha=0) {
    bClampLinearTranslationLimitToRefPose=true
    bOverrideWorldGravity=true
    overridePhysicsAsset=&M_MED_Banner_Physics
    simulationSpace=BaseBoneSpace
    alphaBoolBlend {blendOption=Linear}
    alphaScaleBiasClamp {bInterpResult=true, interpSpeedIncreasing=7}
    baseBoneRef {boneName=pelvis}
    componentAppliedLinearAccClamp=<4000 4000 4000>
    componentLinearAccScale=<0.4 0.4 0.4>
    componentLinearVelScale=<0.4 0.4 0.4>
    overrideWorldGravity {z=-980}
   }
   Output Pose(Result=Blend Poses by bool(BlendPose_0=Component To Local(ComponentPose=@n15), BlendPose_1=Component To Local(ComponentPose=@n14), bActiveValue=Get bIsCrouching()) {blendType=Linear}) {layerGroup=None}
   n16 = Pose Driver - Source: thigh_l!(SourcePose:Pose Link Structure=Copy Pose From Mesh() {bUseAttachedParent=true}) -> (Pose:Pose Link Structure) {
    is=title AnimGraphNode_PoseDriver
    poseAsset=&/M_MED_Banner_LegMovement_PoseAsset
    onlyDriveBones [{boneName=dyn_belt_05}, {boneName=dyn_belt_04}]
    poseTargets [
     {
      bApplyCustomCurve=false
      bIsHidden=false
      distanceMethod=DefaultMethod
      drivenName=default
      functionType=DefaultFunction
      targetScale=1
      boneTransforms [{targetRotation=<rot 6.5552 -1.7725 0.0405>, targetTranslation=<-2.5693 -0.0109 -7.7501>}]
      customCurve {defaultValue=3.4028234663852886e+38, postInfinityExtrap=RCCE_Constant, preInfinityExtrap=RCCE_Constant, keys []}
      targetRotation=<rot 0 0 0>
     }
     {
      bApplyCustomCurve=false
      bIsHidden=false
      distanceMethod=DefaultMethod
      drivenName=upleg_upMax_l
      functionType=DefaultFunction
      targetScale=0.2
      boneTransforms [{targetRotation=<rot -1.0963 -101.713 -6.4634>, targetTranslation=<-2.5693 -0.0109 -7.7501>}]
      customCurve {defaultValue=3.4028234663852886e+38, postInfinityExtrap=RCCE_Constant, preInfinityExtrap=RCCE_Constant, keys []}
      targetRotation=<rot 0 0 0>
     }
    ]
    rBFParams {radius=400}
    sourceBones [{boneName=thigh_l}]
   }
   n17 = Pose Driver - Source: thigh_r!(SourcePose:Pose Link Structure=@n16) -> (Pose:Pose Link Structure) {
    is=title AnimGraphNode_PoseDriver
    poseAsset=&M_MED_Banner_LegMovement_PoseAsset
    onlyDriveBones [{boneName=dyn_belt_01}, {boneName=dyn_belt_02}, {boneName=dyn_belt_06}, {boneName=dyn_belt_07}, {boneName=dyn_belt_03}]
    poseTargets [
     {
      bApplyCustomCurve=false
      bIsHidden=false
      distanceMethod=DefaultMethod
      drivenName=default
      functionType=DefaultFunction
      targetScale=0.001
      boneTransforms [{targetRotation=<rot 6.5552 178.2275 0.0405>, targetTranslation=<-2.5691 -0.0109 7.7501>}]
      customCurve {defaultValue=3.4028234663852886e+38, postInfinityExtrap=RCCE_Constant, preInfinityExtrap=RCCE_Constant, keys []}
      targetRotation=<rot 0 0 0>
     }
     {
      bApplyCustomCurve=false
      bIsHidden=false
      distanceMethod=DefaultMethod
      drivenName=upleg_upMid_r
      functionType=DefaultFunction
      targetScale=1
      boneTransforms [{targetRotation=<rot 4.6587 133.0373 -4.6169>, targetTranslation=<-2.5691 -0.0109 7.7501>}]
      customCurve {defaultValue=3.4028234663852886e+38, postInfinityExtrap=RCCE_Constant, preInfinityExtrap=RCCE_Constant, keys []}
      targetRotation=<rot 0 0 0>
     }
     {
      bApplyCustomCurve=false
      bIsHidden=false
      distanceMethod=DefaultMethod
      drivenName=upleg_side_r
      functionType=DefaultFunction
      targetScale=0.5195
      boneTransforms [{targetRotation=<rot 51.5552 178.2735 0.0647>, targetTranslation=<-2.5691 -0.0109 7.7501>}]
      customCurve {defaultValue=3.4028234663852886e+38, postInfinityExtrap=RCCE_Constant, preInfinityExtrap=RCCE_Constant, keys []}
      targetRotation=<rot 0 0 0>
     }
     {
      bApplyCustomCurve=false
      bIsHidden=false
      distanceMethod=DefaultMethod
      drivenName=upleg_upSideMid_r
      functionType=DefaultFunction
      targetScale=1
      boneTransforms [{targetRotation=<rot 31.1235 125.3881 4.167>, targetTranslation=<-2.5691 -0.0109 7.7501>}]
      customCurve {defaultValue=3.4028234663852886e+38, postInfinityExtrap=RCCE_Constant, preInfinityExtrap=RCCE_Constant, keys []}
      targetRotation=<rot 0 0 0>
     }
     {
      bApplyCustomCurve=false
      bIsHidden=false
      distanceMethod=DefaultMethod
      drivenName=upleg_upSideMax_r
      functionType=DefaultFunction
      targetScale=0.9017
      boneTransforms [{targetRotation=<rot 23.0447 71.6027 -21.9004>, targetTranslation=<-2.5691 -0.0109 7.7501>}]
      customCurve {defaultValue=3.4028234663852886e+38, postInfinityExtrap=RCCE_Constant, preInfinityExtrap=RCCE_Constant, keys []}
      targetRotation=<rot 0 0 0>
     }
     {
      bApplyCustomCurve=false
      bIsHidden=false
      distanceMethod=DefaultMethod
      drivenName=upleg_upMax_r
      functionType=DefaultFunction
      targetScale=0.9017
      boneTransforms [{targetRotation=<rot -1.5228 74.4968 -6.3767>, targetTranslation=<-2.5691 -0.0109 7.7501>}]
      customCurve {defaultValue=3.4028234663852886e+38, postInfinityExtrap=RCCE_Constant, preInfinityExtrap=RCCE_Constant, keys []}
      targetRotation=<rot 0 0 0>
     }
    ]
    rBFParams {radius=40}
    sourceBones [{boneName=thigh_r}]
   }
   n18 = "Constraint - dyn_belt_07 (PP)"!(ComponentPose=Local To Component(LocalPose=@n17), Alpha=1, ConstraintWeights_0=1, ConstraintWeights_1:float=1) -> (Pose) {
    alphaBoolBlend {blendOption=Linear}
    boneToModify {boneName=dyn_belt_07}
    constraintSetup [{targetBone {boneName=dyn_belt_02}}, {targetBone {boneName=dyn_belt_03}}]
    constraintWeights [1,1]
   }
   Save cached pose 'BeltCache'(Pose=Component To Local(ComponentPose=@n18)) {cacheName=BeltCache}
  ]}
 EventGraph {nodes [
   Event BlueprintUpdateAnimation()
   > n1 = Sequence()
   > Set Crouching_Alpha(Crouching_Alpha:double=SelectFloat(A=1, B=0, bPickA=Get bIsCrouching())) -> (Output_Get:double)
   > Set Ads_Alpha(Ads_Alpha:double=SelectFloat(A=1, B=0, bPickA=Get bIsTargeting())) -> (Output_Get:double)
   n2 = Get bIsFalling()
   n3 = Get bIsJumping()
   > Set Jump_Alpha(Jump_Alpha:double=SelectFloat(A=0.6, B=0, bPickA=OR Boolean!(A=@n2, B=@n3) -> (ReturnValue))) -> (Output_Get:double)
   n4 = Branch(execute=@n1.then_1, Condition=Get bIsJumping())
   > Delay(Duration=0.25)
   > Set Belt_ext_force(Belt_ext_force:Vector="0,0,1250") -> (Output_Get:Vector)
   > n5 = Delay(Duration=0.375)
   Do Once(execute=@n4.else, Reset=@n5.then)
   > Set Belt_ext_force(Belt_ext_force="0, 0, 0")
   TryGetPawnOwner()
  ]}
]
variables [
 Ads_Alpha {type=number}
 Belt_ext_force {type=Vector}
 Crouching_Alpha {type=number}
 Jump_Alpha {type=number}
]
